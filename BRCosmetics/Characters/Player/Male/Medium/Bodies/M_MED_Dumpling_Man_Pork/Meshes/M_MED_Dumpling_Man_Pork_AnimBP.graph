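# Post-process style cosmetic AnimBP for the Pork Dumpling Man body, adding secondary motion to skirt, jar, spat and pouch bones.
# Copies the parent mesh pose, offsets the jar by right-thigh lift, then runs RigidBody physics, bone constraints and a spat Look At.
# Crouching blends in a back-pouch offset before a final Control Rig.
# Events update thigh angles and reset dynamics around skydiving, parachuting, crouching and targeting.
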AnimBlueprint /BRCosmetics/Characters/Player/Male/Medium/Bodies/M_MED_Dumpling_Man_Pork/Meshes/M_MED_Dumpling_Man_Pork_AnimBP
target_skeleton=&/M_MED_Dumpling_Man_Pork_Skeleton
graphs [
 AnimGraph {nodes [
   Control Rig(Alpha=0) {
    alphaBoolBlend {blendOption=Linear}
    controlRigAssetReference {blueprintRigClass=&/M_MED_Dumpling_Man_Pork_CtrlRig.M_MED_Dumpling_Man_Pork_CtrlRig_C}
   }
   "Transform (Modify) Bone - Bone: dyn_skirt_bk_r_01"!(ComponentPose, Alpha=1, Translation="0,5,0", Rotation="0,0,0") -> (Pose) {
    rotationMode=BMM_Additive
    rotationSpace=BCS_BoneSpace
    translationMode=BMM_Additive
    translationSpace=BCS_BoneSpace
    alphaBoolBlend {blendOption=Linear}
    boneToModify {boneName=dyn_skirt_bk_r_01}
    translation {y=5}
   }
   n1 = "Transform (Modify) Bone - Bone: dyn_skirt_bk_l_01"!(ComponentPose=Local To Component(LocalPose=Use cached pose 'dynCache'()), Alpha=0, Translation="-5,-5,10", Rotation="20,-10,0") -> (Pose) {
    rotationMode=BMM_Additive
    rotationSpace=BCS_BoneSpace
    translationMode=BMM_Additive
    translationSpace=BCS_BoneSpace
    alphaBoolBlend {blendOption=Linear}
    boneToModify {boneName=dyn_skirt_bk_l_01}
    rotation {pitch=20, yaw=-10}
    translation=<-5 -5 10>
   }
   n2 = "Transform (Modify) Bone - Bone: dyn_skirt_bk_r_01"(ComponentPose=@n1, Alpha=0, Translation="2,0,-5", Rotation="20,-10,0") {
    rotationMode=BMM_Additive
    rotationSpace=BCS_BoneSpace
    translationMode=BMM_Additive
    translationSpace=BCS_BoneSpace
    alphaBoolBlend {blendOption=Linear}
    boneToModify {boneName=dyn_skirt_bk_r_01}
    rotation {pitch=20, yaw=-10}
    translation {x=2, z=-5}
   }
   n3 = "Transform (Modify) Bone - Bone: dyn_pouch_Back"!(ComponentPose=@n2, Alpha=1, Translation="0,0,0", Rotation="0,-50,-30") -> (Pose) {
    rotationMode=BMM_Additive
    rotationSpace=BCS_BoneSpace
    translationMode=BMM_Additive
    translationSpace=BCS_BoneSpace
    alphaBoolBlend {blendOption=Linear}
    boneToModify {boneName=dyn_pouch_Back}
    rotation {roll=-30, yaw=-50}
   }
   n4 = Control Rig(Source=Blend Poses by bool(BlendPose_0=Component To Local(ComponentPose=@n3), BlendPose_1=Use cached pose 'dynCache'(), bActiveValue=Get bIsCrouching()) {blendType=Linear}) {
    alphaBoolBlend {blendOption=Linear}
    controlRigAssetReference {blueprintRigClass=&M_MED_Dumpling_Man_Pork_CtrlRig_C}
   }
   Output Pose(Result=@n4) {layerGroup=None}
   Save cached pose 'cpfm Cache'(Pose=Copy Pose From Mesh() {bUseAttachedParent=true}) {cacheName=cpfm Cache}
   n5 = "Transform (Modify) Bone - Bone: dyn_jar_01_base"(
    ComponentPose=Local To Component(LocalPose=Use cached pose 'cpfm Cache'())
    Alpha=Get legUp_R() -> (legUp_R:double)
    Translation="3,-10,0"
    Rotation="-40,0,-60"
    Scale="1,1,1"
   ) {
    rotationMode=BMM_Additive
    rotationSpace=BCS_BoneSpace
    translationMode=BMM_Additive
    translationSpace=BCS_BoneSpace
    alphaBoolBlend {blendOption=Linear}
    alphaScaleBiasClamp {bMapRange=true, inRange {max=0, min=-100}, outRange {max=0, min=1}}
    boneToModify {boneName=dyn_jar_01_base}
    rotation {pitch=-40, roll=-60}
    translation=<3 -10>
   }
   n6 = RigidBody(ComponentPose=@n5) {
    bClampLinearTranslationLimitToRefPose=true
    bOverrideWorldGravity=true
    overridePhysicsAsset=&/M_MED_Dumpling_Pork_Body
    simulationSpace=BaseBoneSpace
    alphaBoolBlend {blendOption=Linear}
    baseBoneRef {boneName=pelvis}
    componentAppliedLinearAccClamp=<2000 2000 2000>
    componentLinearAccScale=<0.1 0.1 0.1>
    componentLinearVelScale=<0.1 0.1 0.1>
    overrideWorldGravity {z=-980}
   }
   n7 = "Constraint - dyn_skirt_fr_01 (PP)"!(ComponentPose=@n6, Alpha=1, ConstraintWeights_0=0.5, ConstraintWeights_1:float=0.5) -> (Pose) {
    alphaBoolBlend {blendOption=Linear}
    boneToModify {boneName=dyn_skirt_fr_01}
    constraintSetup [{targetBone {boneName=dyn_skirt_fr_l_01}}, {targetBone {boneName=dyn_skirt_fr_r_01}}]
    constraintWeights [1,1]
   }
   n8 = Anim Dynamics - Bone: dyn_jar_02(ComponentPose=@n7, Alpha=0, ExternalForce="0,0,0") {
    angularDampingOverride=0.999
    bGravityOverrideInSimSpace=true
    bOverrideAngularDamping=true
    gravityScale=0.35
    simulationSpace=RootRelative
    alphaBoolBlend {blendOption=Linear}
    boundBone {boneName=dyn_jar_02}
    componentLinearAccScale=<0.85 0.85 0.85>
    componentLinearVelScale=<0.85 0.85 0.85>
    physicsBodyDefinitions [{
      sphereCollisionRadius=0
      boundBone {boneName=dyn_jar_02}
      boxExtents=<5 20 20>
      constraintSetup {angularLimitsMax=<0 15 5>, angularLimitsMin=<0 -15 -5>, angularTarget=<0 0 0>}
      localJointOffset=<0 1 0>
     }]
   }
   n9 = "Transform (Modify) Bone - Bone: dyn_jar_02"(ComponentPose=@n8, Alpha=1, Translation="-1,0,1", Rotation="-10,0,0", Scale="1,1,1") {
    rotationMode=BMM_Additive
    rotationSpace=BCS_BoneSpace
    translationMode=BMM_Additive
    translationSpace=BCS_BoneSpace
    alphaBoolBlend {blendOption=Linear}
    alphaScaleBiasClamp {inRange {max=0, min=-100}, outRange {max=0, min=1}}
    boneToModify {boneName=dyn_jar_02}
    rotation {pitch=-10}
    translation {x=-1, z=1}
   }
   n10 = "Constraint - const_spat_target (PPPPPP)"!(
    ComponentPose=@n9
    Alpha=1
    ConstraintWeights_0=0.5605
    ConstraintWeights_1:float=0.3234
    ConstraintWeights_2:float=0.0155
    ConstraintWeights_3:float=0.0024
    ConstraintWeights_4:float=0.0909
    ConstraintWeights_5:float=0.0073
   ) -> (Pose) {
    alphaBoolBlend {blendOption=Linear}
    boneToModify {boneName=const_spat_target}
    constraintSetup [
     {targetBone {boneName=dyn_skirt_fr_r_01}}
     {targetBone {boneName=dyn_skirt_fr_01}}
     {targetBone {boneName=dyn_skirt_fr_l_01}}
     {targetBone {boneName=thigh_twist_01_r}}
     {targetBone {boneName=dyn_skirt_out_r_01}}
     {targetBone {boneName=pelvis}}
    ]
    constraintWeights [1,1,1,1,1,1]
   }
   n11 = Look At - Bone: dyn_spat_02(ComponentPose=@n10, Alpha=1) {
    alphaBoolBlend {blendOption=Linear}
    boneToModify {boneName=dyn_spat_02}
    lookAtLocation {x=0}
    lookAtTarget {boneReference {boneName=const_spat_target}}
    lookAt_Axis {axis=<-1 0>}
   }
   Save cached pose 'dynCache'(Pose=Component To Local(ComponentPose=@n11)) {cacheName=dynCache}
   Get Iscrouchinganim() -> (Iscrouchinganim:double)
   Get Iscrouchinganim()
   Get Iscrouchinganim()
  ]}
 EventGraph {nodes [
   Event BlueprintUpdateAnimation()
   > n1 = Sequence!(execute) -> (then_0, then_1, then_2:Exec)
   Set Iscrouchinganim(execute=@n1.then_2, Iscrouchinganim:double=SelectFloat(A=1, B=0, bPickA=Get bIsCrouching())) -> (Output_Get:double)
   n2 = GetOwnerSkeletalMeshComponent()
   n3 = GetDeltaTransformFromRefPose(self=@n2, BoneName=thigh_l, BaseName=pelvis)
   Set legUp_L(execute=@n1.then_0, legUp_L:double=BreakRotator(InRot=BreakTransform(InTransform=@n3).Rotation).Yaw) -> (Output_Get:double)
   n4 = BreakRotator(InRot=BreakTransform(InTransform=GetDeltaTransformFromRefPose(self=@n2, BoneName=thigh_r, BaseName=pelvis)).Rotation)
   Set legUp_R(execute=@n1.then_1, legUp_R:double=@n4.Yaw) -> (Output_Get:double)
   > Set legSide_R(legSide_R:double=@n4.Pitch) -> (Output_Get:double)
   Event OnParachuteOpened()
   > n5 = Delay(Duration=1.5)
   ResetDynamics(execute=[@n5.then, @n6.then, @n7.then, @n10.then], InTeleportType=TeleportPhysics)
   > ResetDynamics(InTeleportType=ResetPhysics)
   n6 = Event OnBeginSkydiving()
   n7 = Event OnEndSkydiving()
   n8 = Event OnEndCrouching()
   n9 = Get bIsCrouching()
   n10 = Branch(execute=[@n8.then, @n11.then], Condition=@n9)
   n11 = Event OnBeginTargeting() {is=title K2Node_Event}
   AND Boolean(A=Get bIsTargeting(), B=NOT Boolean(A=@n9))
  ]}
]
variables [
 Iscrouchinganim {type=number}
 legSide_R {type=number}
 legUp_L {type=number}
 legUp_R {type=number}
]
