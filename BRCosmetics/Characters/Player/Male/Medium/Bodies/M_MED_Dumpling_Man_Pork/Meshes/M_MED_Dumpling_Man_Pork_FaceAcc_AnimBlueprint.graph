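# Animation Blueprint for the face accessory of the Dumpling Man Pork player cosmetic, adding secondary motion to its bow and flop bones.
# It copies the pose from the attached parent mesh, then runs two chained, angle-limited Anim Dynamics nodes in component space.
# Both share an alpha that rises with 2D pawn speed and drops to zero while targeting.
# The event graph holds no connected logic.
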AnimBlueprint /BRCosmetics/Characters/Player/Male/Medium/Bodies/M_MED_Dumpling_Man_Pork/Meshes/M_MED_Dumpling_Man_Pork_FaceAcc_AnimBlueprint
target_skeleton=&/M_MED_Dumpling_Man_Pork_FaceAcc_Skeleton
graphs [
 AnimGraph {nodes [
   n1 = MapRangeClamped(Value=Get PawnSpeed2D(), InRangeA=0, InRangeB=200, OutRangeB=SelectFloat(A=0, B=0.85, bPickA=Get bIsTargeting()))
   n2 = Anim Dynamics - Bone: dyn_bow_mstr(ComponentPose=Local To Component(LocalPose=Copy Pose From Mesh() {bUseAttachedParent=true}), Alpha=@n1, ExternalForce="0,0,0") {
    angularDampingOverride=0.98
    bGravityOverrideInSimSpace=true
    bOverrideAngularBias=true
    bOverrideLinearDamping=true
    gravityScale=0.65
    lODThreshold=1
    linearDampingOverride=0.98
    numSolverIterationsPostUpdate=4
    numSolverIterationsPreUpdate=16
    alphaBoolBlend {blendOption=Linear}
    boundBone {boneName=dyn_bow_mstr}
    physicsBodyDefinitions [{
      sphereCollisionRadius=0
      boundBone {boneName=dyn_bow_mstr}
      boxExtents=<5 5 5>
      constraintSetup {angularTargetAxis=AxisZ, twistAxis=AxisZ, angularLimitsMax=<45 0 10>, angularLimitsMin=<-45 -30 -10>, angularTarget=<0 0 0>}
      localJointOffset=<0 0 -1>
     }]
   }
   n3 = Anim Dynamics - Bone: dyn_flop_01(ComponentPose=@n2, Alpha=@n1, ExternalForce="0,0,0") {
    angularDampingOverride=0.98
    bGravityOverrideInSimSpace=true
    bOverrideAngularBias=true
    bOverrideLinearDamping=true
    gravityScale=0.65
    lODThreshold=1
    linearDampingOverride=0.98
    numSolverIterationsPostUpdate=4
    numSolverIterationsPreUpdate=16
    alphaBoolBlend {blendOption=Linear}
    boundBone {boneName=dyn_flop_01}
    physicsBodyDefinitions [{
      sphereCollisionRadius=0
      boundBone {boneName=dyn_flop_01}
      boxExtents=<5 5 5>
      constraintSetup {angularTargetAxis=AxisZ, twistAxis=AxisZ, angularLimitsMax=<0 30 0>, angularLimitsMin=<0 -30 0>, angularTarget=<0 0 0>}
      localJointOffset=<0 0 -1>
     }]
   }
   Output Pose(Result=Component To Local(ComponentPose=@n3)) {layerGroup=None}
  ]}
 EventGraph {nodes [
   TryGetPawnOwner()
   Event BlueprintUpdateAnimation()
  ]}
]
variables []
